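; Cage Saddle backpack character-part Blueprint that plays a poof effect on spawn and drives a status light showing block deployment state.
; After parent init, async gameplay-tag tasks watch the pawn's deployment and cooldown tags to turn the light red or green, consulting the experience's NoCooldowns setting.
; A part-applied callback caches a dynamic material, whose emissive an example timeline lerps in events marked for deletion.
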(event EventBeginPlay
  (Utilities|FlowControl|DelayUntilNextTick)
  (Effects|SpawnFXSystemAttached :FXSystemAsset "/MotherGameplay/Items/CageSaddle/Assets/FX/NS_BackpackPoof.NS_BackpackPoof" :AttachToComponent (Variables|/Script/Engine.Actor|GetRootComponent (CustomCharacter|GetAssociatedPlayerPawn)) :AttachPointName "None" :Location "-30.000000, 0,55.000000" :Scale "1.000000,1.000000,1.000000" :LocationType "SnapToTarget" :bAutoDestroy true :PoolingMethod "None" :bAutoActivate true))

(event Collision|EventActorBeginOverlap
  (params (OtherActor :type Actor%20Object%20Reference)))

(event EventTick
  (params (DeltaSeconds :type Float%20%28single-precision%29)))

(event EventOnPostResIn
  ;[1 1 1] Init
  (|CallParentFunction)
  (Utilities|FlowControl|Sequence
    (:then_0
      ;; AddWaitTasks
      (bind _asynctaskproxy (Ability|Tasks|WaitGameplayTagAddToActor (CustomCharacter|GetAssociatedPlayerPawn) "(TagName=\"Granted.Athena.CageSaddle.BlockDeployment\")")
        (:then
          (Variables|Default|SetWaitForBlockAdded _asynctaskproxy)
          (bind _asynctaskproxy_2 (Ability|Tasks|WaitGameplayTagRemoveFromActor (CustomCharacter|GetAssociatedPlayerPawn) "(TagName=\"Granted.Athena.CageSaddle.BlockDeployment\")")
            (:then
              (Variables|Default|SetWaitForBlockRemoved _asynctaskproxy_2)
              (Variables|Default|SetExperienceSettingsComp (Fortnite|Athena|GameState|GetExperienceSettingsComponentfromGameState))
              (bind _asynctaskproxy_3 (Ability|Tasks|WaitGameplayTagAddToActor (CustomCharacter|GetAssociatedPlayerPawn) "(TagName=\"Granted.Cooldown.CageSaddle.BlockDeployment\")")
                (:then
                  (Variables|Default|SetWaitForCooldownAdded _asynctaskproxy_3)
                  (Utilities|IsValid (Variables|Default|GetExperienceSettingsComp)
                    (:Is%20Valid
                      _)
                    (:Is%20Not%20Valid
                      (bind _asynctaskproxy_4 (Ability|Tasks|WaitGameplayTagRemoveFromActor (CustomCharacter|GetAssociatedPlayerPawn) "(TagName=\"Granted.Cooldown.CageSaddle.BlockDeployment\")")
                        (:then
                          (Variables|Default|SetWaitForCooldownRemoved _asynctaskproxy_4))
                        (:Removed
                          (CallFunction|TurnLightGreen))))))
                (:Added
                  (Utilities|IsValid (Variables|Default|GetExperienceSettingsComp)
                    (:Is%20Valid
                      (if (not (ScalableFloat|GetValueasBoolwithContextParam (CustomCharacter|GetAssociatedPlayerPawn) (Variables|/Script/FortniteGame.FortExperienceSettingsComponent|GetNoCooldowns (Variables|Default|GetExperienceSettingsComp))))
                        (CallFunction|TurnLightRed)
                        (Utilities|IsValid (Variables|Default|GetWaitForCooldownRemoved)
                          (:Is%20Not%20Valid
                            (bind _asynctaskproxy_5 (Ability|Tasks|WaitGameplayTagRemoveFromActor (CustomCharacter|GetAssociatedPlayerPawn) "(TagName=\"Granted.Cooldown.CageSaddle.BlockDeployment\")" true)
                              (:then
                                (Variables|Default|SetWaitForCooldownRemoved _asynctaskproxy_5))
                              (:Removed
                                (CallFunction|TurnLightGreen)
                                (Variables|Default|SetWaitForCooldownRemoved 0)))))))
                    (:Is%20Not%20Valid
                      (CallFunction|TurnLightRed))))))
            (:Removed
              (CallFunction|TurnLightGreen))))
        (:Added
          (CallFunction|TurnLightRed))))
    (:then_1
      (bind _delegatehandlecontroller_out (Cosmetics|WhenCosmeticSpecificPartApplied (CustomCharacter|GetAssociatedPlayerPawn) (delegate Custom|SetupDMI) true (CustomCharacter|GetCharacterPartType)))
      (Variables|Default|SetWaitForBackpackApplied _delegatehandlecontroller_out))))

(event Custom|SetupDMI
  (params (Pawn :type Fort%20Player%20Pawn%20Object%20Reference) (Part :type Custom%20Character%20Part%20Object%20Reference) (MeshPart :type Skeletal%20Mesh%20Component%20Object%20Reference) (PartType :type EFortCustomPartType%20Enum :default "Head"))
  ;[1 1 1] Init
  (Utilities|IsValid (Variables|Default|GetMaterial)
    (:Is%20Not%20Valid
      (bind _returnvalue (Fortnite|GetDynamicMaterialInstance MeshPart))
      (bind _output_get (Variables|Default|SetMaterial _returnvalue))
      (bind _returnvalue_2 (Rendering|Material|GetVectorParameterValue _output_get "EmissiveMultiplier" :!class /Script/Engine.MaterialInstanceDynamic))
      (Variables|Default|SetEmissive _returnvalue_2))))

(event Custom|TurnLightGreen_
  ;[1 1 1] Backpack Light Example (Delete once actual event is hooked up.)
  (GameplayCue|ExecuteGameplayCue%28Burst%29Local :Target (CustomCharacter|GetAssociatedPlayerPawn) :CueParameters (Ability|GameplayCue|MakeGameplayCueParameters 0.0) :bOnlyNotifies true)
  (Utilities|IsValid (Variables|Default|GetMaterial)
    (:Is%20Valid
      (bind (_direction _emissiveinterpolator) (Timeline|EmissiveTest PlayFromStart)
        (:Update
          (Rendering|Material|SetVectorParameterValue (Variables|Default|GetMaterial) "EmissiveMultiplier" (Math|Color|Lerp(LinearColor) (Variables|Default|GetEmissive) (Utilities|Struct|MakeLinearColor 8.0 0.0 0.0 1.0) _emissiveinterpolator) :!class /Script/Engine.MaterialInstanceDynamic))))))

(event Custom|TurnLightRed_
  ;[1 1 1] Backpack Light Example (Delete once actual event is hooked up.)
  (Utilities|IsValid (Variables|Default|GetMaterial)
    (:Is%20Valid
      (Timeline|EmissiveTest ReverseFromEnd))))

(fn ConstructionScript)
