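# Animation Blueprint for the Cage Saddle back gadget, posing its inflatable creature-like body and adding secondary motion.
# A copied mesh pose is layered per bone with a "deflated" pose by InflationBlendWeight, then KnobRotation and HeadRotation rotate those bones additively.
# Chained spring AnimDynamics nodes then jiggle arms, legs, head, left ear (simulated twice, no right), pullouts, tails and tube.
# The EventGraph is empty, so driving values are set outside it.
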AnimBlueprint /MotherGameplay/Items/CageSaddle/Assets/Cage_Saddle_Back_Gadget_AnimBP
target_skeleton=&/Cage_Saddle_Back_Gadget_Skeleton
graphs [
 AnimGraph {nodes [
   n1 = Layered blend per bone(
    BasePose=Copy Pose From Mesh()
    BlendPoses_0=Cage_Saddle_Back_Poses_PoseAsset : deflated!(PoseWeight:float=1) -> (Pose:Pose Link Structure) {
     is=title AnimGraphNode_PoseByName
     poseAsset=&/Cage_Saddle_Back_Poses_PoseAsset
     poseName=deflated
     poseWeight=0
    }
    BlendWeights_0=Get InflationBlendWeight() -> (InflationBlendWeight:float)
   ) {layerSetup [{branchFilters [
       {boneName=cageSaddle_body_offset}
       {boneName=cageSaddle_head}
       {boneName=cageSaddle_tail_a}
       {boneName=cageSaddle_arm_l}
       {boneName=cageSaddle_arm_r}
       {boneName=cageSaddle_leg_l}
       {boneName=cageSaddle_leg_r}
       {boneName=cageSaddle_tail_b}
       {boneName=cageSaddle_pullout_a}
       {boneName=cageSaddle_eye_l}
       {boneName=cageSaddle_eye_r}
       {boneName=cageSaddle_ear_l}
       {boneName=cageSaddle_ear_r}
       {boneName=cageSaddle_horn_l}
       {boneName=cageSaddle_horn_r}
       {boneName=cageSaddle_mouth}
       {boneName=cageSaddle_tube}
      ]}]}
   n2 = "Transform (Modify) Bone - Bone: cageSaddle_knob"(
    ComponentPose=Local To Component(LocalPose=@n1)
    Alpha=1
    Translation="0,0,0"
    Rotation=Get KnobRotation() -> (KnobRotation:Rotator)
    Scale="1,1,1"
   ) {
    rotationMode=BMM_Additive
    rotationSpace=BCS_BoneSpace
    translationMode=BMM_Additive
    translationSpace=BCS_BoneSpace
    alphaBoolBlend {blendOption=Linear}
    boneToModify {boneName=cageSaddle_knob}
   }
   n3 = "Transform (Modify) Bone - Bone: cageSaddle_head_base"(ComponentPose=@n2, Alpha=1, Translation="0,0,0", Rotation=Get HeadRotation() -> (HeadRotation:Rotator), Scale="1,1,1") {
    rotationMode=BMM_Additive
    rotationSpace=BCS_BoneSpace
    translationMode=BMM_Additive
    translationSpace=BCS_BoneSpace
    alphaBoolBlend {blendOption=Linear}
    boneToModify {boneName=cageSaddle_head_base}
    rotation {roll=-10}
   }
   n4 = Anim Dynamics - Bone: cageSaddle_arm_base_l(ComponentPose=@n3, Alpha=1, ExternalForce="0,0,0") {
    angularDampingOverride=1
    angularSpringConstant=1500
    bAngularSpring=true
    bGravityOverrideInSimSpace=true
    bOverrideAngularDamping=true
    gravityScale=0.35
    simulationSpace=RootRelative
    alphaBoolBlend {blendOption=Linear}
    boundBone {boneName=cageSaddle_arm_base_l}
    componentLinearAccScale=<1 1 0.5>
    componentLinearVelScale=<1 1 0.5>
    physicsBodyDefinitions [{
      sphereCollisionRadius=0
      boundBone {boneName=cageSaddle_arm_base_l}
      boxExtents=<20 10 10>
      constraintSetup {angularLimitsMax=<0 20 20>, angularLimitsMin=<0 -20 -20>, angularTarget=<0 0 0>}
      localJointOffset=<5 0 0>
     }]
   }
   n5 = Anim Dynamics - Bone: cageSaddle_arm_base_r(ComponentPose=@n4, Alpha=1, ExternalForce="0,0,0") {
    angularDampingOverride=1
    angularSpringConstant=1500
    bAngularSpring=true
    bGravityOverrideInSimSpace=true
    bOverrideAngularDamping=true
    gravityScale=0.35
    simulationSpace=RootRelative
    alphaBoolBlend {blendOption=Linear}
    boundBone {boneName=cageSaddle_arm_base_r}
    componentLinearAccScale=<0.8 0.8 0.4>
    componentLinearVelScale=<0.8 0.8 0.4>
    physicsBodyDefinitions [{
      sphereCollisionRadius=0
      boundBone {boneName=cageSaddle_arm_base_r}
      boxExtents=<20 10 10>
      constraintSetup {angularLimitsMax=<0 20 20>, angularLimitsMin=<0 -20 -20>, angularTarget=<0 0 0>}
      localJointOffset=<-5 0 0>
     }]
   }
   n6 = Anim Dynamics - Bone: cageSaddle_base_leg_base_l(ComponentPose=@n5, Alpha=1, ExternalForce="0,0,0") {
    angularDampingOverride=1
    angularSpringConstant=1500
    bAngularSpring=true
    bGravityOverrideInSimSpace=true
    bOverrideAngularDamping=true
    gravityScale=0.35
    simulationSpace=RootRelative
    alphaBoolBlend {blendOption=Linear}
    boundBone {boneName=cageSaddle_base_leg_base_l}
    componentLinearAccScale=<0.7 0.7 0.35>
    componentLinearVelScale=<0.7 0.7 0.35>
    physicsBodyDefinitions [{
      sphereCollisionRadius=0
      boundBone {boneName=cageSaddle_base_leg_base_l}
      boxExtents=<20 10 10>
      constraintSetup {angularLimitsMax=<0 20 20>, angularLimitsMin=<0 -20 -20>, angularTarget=<0 0 0>}
      localJointOffset=<5 0 0>
     }]
   }
   n7 = Anim Dynamics - Bone: cageSaddle_leg_base_r(ComponentPose=@n6, Alpha=1, ExternalForce="0,0,0") {
    angularDampingOverride=1
    angularSpringConstant=1500
    bAngularSpring=true
    bGravityOverrideInSimSpace=true
    bOverrideAngularDamping=true
    gravityScale=0.35
    simulationSpace=RootRelative
    alphaBoolBlend {blendOption=Linear}
    boundBone {boneName=cageSaddle_leg_base_r}
    componentLinearAccScale=<0.9 0.9 0.45>
    componentLinearVelScale=<0.9 0.9 0.45>
    physicsBodyDefinitions [{
      sphereCollisionRadius=0
      boundBone {boneName=cageSaddle_leg_base_r}
      boxExtents=<20 10 10>
      constraintSetup {angularLimitsMax=<0 20 20>, angularLimitsMin=<0 -20 -20>, angularTarget=<0 0 0>}
      localJointOffset=<-5 0 0>
     }]
   }
   n8 = Anim Dynamics - Bone: cageSaddle_head_base(ComponentPose=@n7, Alpha=0.7, ExternalForce="0,0,0") {
    angularDampingOverride=1
    angularSpringConstant=1500
    bAngularSpring=true
    bGravityOverrideInSimSpace=true
    bOverrideAngularDamping=true
    gravityScale=0.35
    simulationSpace=RootRelative
    alphaBoolBlend {blendOption=Linear}
    boundBone {boneName=cageSaddle_head_base}
    componentLinearAccScale=<0.5 0.5 0.4>
    componentLinearVelScale=<0.5 0.5 0.4>
    physicsBodyDefinitions [{
      sphereCollisionRadius=0
      boundBone {boneName=cageSaddle_head_base}
      boxExtents=<20 20 10>
      constraintSetup {angularTargetAxis=AxisZ, twistAxis=AxisZ, angularLimitsMax=<5 10 15>, angularLimitsMin=<-20 -10 -15>, angularTarget=<0 0 0>}
      localJointOffset=<0 0 5>
     }]
   }
   n9 = Anim Dynamics - Bone: cageSaddle_ear_base_l(ComponentPose=@n8, Alpha=1, ExternalForce="0,0,0") {
    angularDampingOverride=1
    angularSpringConstant=1500
    bAngularSpring=true
    bGravityOverrideInSimSpace=true
    bOverrideAngularDamping=true
    gravityScale=0.35
    simulationSpace=RootRelative
    alphaBoolBlend {blendOption=Linear}
    boundBone {boneName=cageSaddle_ear_base_l}
    componentLinearAccScale=<1.5 1.5 0.5>
    componentLinearVelScale=<1.5 1.5 0.5>
    physicsBodyDefinitions [{
      sphereCollisionRadius=0
      boundBone {boneName=cageSaddle_ear_base_l}
      boxExtents=<20 20 10>
      constraintSetup {angularLimitsMax=<0 40 15>, angularLimitsMin=<0 -20 -15>, angularTarget=<0 0 0>}
      localJointOffset=<5 0 0>
     }]
   }
   n10 = Anim Dynamics - Bone: cageSaddle_ear_base_l(ComponentPose=@n9, Alpha=1, ExternalForce="0,0,0") {
    angularDampingOverride=1
    angularSpringConstant=1500
    bAngularSpring=true
    bGravityOverrideInSimSpace=true
    bOverrideAngularDamping=true
    gravityScale=0.35
    simulationSpace=RootRelative
    alphaBoolBlend {blendOption=Linear}
    boundBone {boneName=cageSaddle_ear_base_l}
    componentLinearAccScale=<1.5 1.5 0.5>
    componentLinearVelScale=<1.5 1.5 0.5>
    physicsBodyDefinitions [{
      sphereCollisionRadius=0
      boundBone {boneName=cageSaddle_ear_base_l}
      boxExtents=<20 20 10>
      constraintSetup {angularLimitsMax=<0 30 15>, angularLimitsMin=<0 -15 -15>, angularTarget=<0 0 0>}
      localJointOffset=<-5 0 0>
     }]
   }
   n11 = Anim Dynamics - Bone: cageSaddle_pullout_a_base(ComponentPose=@n10, Alpha=1, ExternalForce="0,0,0") {
    angularDampingOverride=1
    angularSpringConstant=1500
    bAngularSpring=true
    bGravityOverrideInSimSpace=true
    bOverrideAngularDamping=true
    bUsePlanarLimit=false
    gravityScale=0.35
    simulationSpace=RootRelative
    alphaBoolBlend {blendOption=Linear}
    boundBone {boneName=cageSaddle_pullout_a_base}
    componentLinearAccScale=<0.35 0.35 0.35>
    componentLinearVelScale=<0.35 0.35 0.35>
    physicsBodyDefinitions [{
      collisionType=CustomSphere
      sphereCollisionRadius=5
      boundBone {boneName=cageSaddle_pullout_a_base}
      boxExtents=<10 20 20>
      constraintSetup {angularLimitsMax=<0 20 10>, angularLimitsMin=<0 -10 -40>, angularTarget=<0 0 0>}
      localJointOffset=<5 0 0>
     }]
   }
   n12 = Anim Dynamics - Bone: cageSaddle_pullout_b_base(ComponentPose=@n11, Alpha=1, ExternalForce="0,0,0") {
    angularDampingOverride=1
    angularSpringConstant=1500
    bAngularSpring=true
    bGravityOverrideInSimSpace=true
    bOverrideAngularDamping=true
    bUsePlanarLimit=false
    gravityScale=0.35
    simulationSpace=RootRelative
    alphaBoolBlend {blendOption=Linear}
    boundBone {boneName=cageSaddle_pullout_b_base}
    componentLinearAccScale=<0.35 0.35 0.35>
    componentLinearVelScale=<0.35 0.35 0.35>
    physicsBodyDefinitions [{
      collisionType=CustomSphere
      sphereCollisionRadius=5
      boundBone {boneName=cageSaddle_pullout_b_base}
      boxExtents=<10 20 20>
      constraintSetup {angularLimitsMax=<10 20 10>, angularLimitsMin=<-10 -10 -40>, angularTarget=<0 0 0>}
      localJointOffset=<5 0 0>
     }]
   }
   n13 = Anim Dynamics - Bone: cageSaddle_tail_a_base(ComponentPose=@n12, Alpha=1, ExternalForce="0,0,0") {
    angularDampingOverride=1
    angularSpringConstant=1500
    bAngularSpring=true
    bGravityOverrideInSimSpace=true
    bOverrideAngularDamping=true
    bUsePlanarLimit=false
    gravityScale=0.35
    simulationSpace=RootRelative
    alphaBoolBlend {blendOption=Linear}
    boundBone {boneName=cageSaddle_tail_a_base}
    componentLinearAccScale=<0.4 0.4 0.4>
    componentLinearVelScale=<0.4 0.4 0.4>
    physicsBodyDefinitions [{
      collisionType=CustomSphere
      sphereCollisionRadius=5
      boundBone {boneName=cageSaddle_tail_a_base}
      boxExtents=<10 20 20>
      constraintSetup {angularLimitsMax=<0 10 0>, angularLimitsMin=<0 -25 -40>, angularTarget=<0 0 0>}
      localJointOffset=<5 0 0>
     }]
   }
   n14 = Anim Dynamics - Bone: cageSaddle_tail_b_base(ComponentPose=@n13, Alpha=1, ExternalForce="0,0,0") {
    angularDampingOverride=1
    angularSpringConstant=1500
    bAngularSpring=true
    bGravityOverrideInSimSpace=true
    bOverrideAngularDamping=true
    bUsePlanarLimit=false
    gravityScale=0.35
    simulationSpace=RootRelative
    alphaBoolBlend {blendOption=Linear}
    boundBone {boneName=cageSaddle_tail_b_base}
    componentLinearAccScale=<0.4 0.4 0.4>
    componentLinearVelScale=<0.4 0.4 0.4>
    physicsBodyDefinitions [{
      collisionType=CustomSphere
      sphereCollisionRadius=5
      boundBone {boneName=cageSaddle_tail_b_base}
      boxExtents=<10 20 20>
      constraintSetup {angularLimitsMax=<0 10 0>, angularLimitsMin=<0 -25 -40>, angularTarget=<0 0 0>}
      localJointOffset=<5 0 0>
     }]
   }
   n15 = Anim Dynamics - Bone: cageSaddle_tube_base(ComponentPose=@n14, Alpha=1, ExternalForce="0,0,0") {
    angularDampingOverride=1
    angularSpringConstant=1500
    bAngularSpring=true
    bGravityOverrideInSimSpace=true
    bOverrideAngularDamping=true
    gravityScale=0.35
    simulationSpace=RootRelative
    alphaBoolBlend {blendOption=Linear}
    boundBone {boneName=cageSaddle_tube_base}
    componentLinearAccScale=<0.3 0.3 0.3>
    componentLinearVelScale=<0.3 0.3 0.3>
    physicsBodyDefinitions [{
      sphereCollisionRadius=0
      boundBone {boneName=cageSaddle_tube_base}
      boxExtents=<10 20 20>
      constraintSetup {angularLimitsMax=<0 5 5>, angularLimitsMin=<0 -5 -5>, angularTarget=<0 0 0>}
      localJointOffset=<5 0 0>
     }]
   }
   Output Pose(Result=Component To Local(ComponentPose=@n15)) {layerGroup=None}
  ]}
 EventGraph {links [], nodes []}
]
variables [SkeletalMeshComponent_old {type=SkeletalMeshComponent}]
