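# Niagara effect for the CageSaddle item, apparently a bounce impact, that bursts rock debris and a ring of dust puffs.
# A ForceIndex module takes User.ForceStrength, and the resulting ForcePicker indexes arrays setting rock launch speed, dust size and ring radius.
# Two mesh rock variants, split by spawn group, launch in a cone, fall, collide and shrink away.
# Dust sprites spread around a torus by spawn group, expanding and fading through curve-driven material parameters.
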NiagaraSystem /MotherGameplay/Items/CageSaddle/Assets/FX/NS_CageSaddle_Poofs
dependencies {
 usedDataInterfaces [&NiagaraDataInterfaceArrayFloat, &NiagaraDataInterfaceColorCurve, &NiagaraDataInterfaceCurve]
 usedDynamicInputs []
 usedModules [&SystemState]
}
emitters {
 Bounce_Area_Rocks001 {
  emitterUpdateScript [
   &EmitterState(Life Cycle Mode=::NewEnumerator1, Loop Behavior=::NewEnumerator1, Loop Duration=1, UseLoopDelay=false)
   /SetVariables() {isSetParameters=true}
   &SpawnBurst_Instantaneous(Spawn Count=5, Use Spawn Probability=true)
   &SpawnBurst_Instantaneous(Spawn Count=4, Spawn Group=1, Use Spawn Probability=true) {module=SpawnBurst_Instantaneous001}
  ]
  particleSpawnScript [
   &self:ForceIndex(ForceStrength:NiagaraFloat=$User.ForceStrength) {module=ForceIndex}
   &FN_SetSpawnGroup(Number of Spawn Groups=10)
   &self:Mesh_Index_01(MeshIndex:NiagaraInt32=$Emitter.SpawnGroup) {module=Mesh_Index_01}
   &InitializeParticle(
    Color Mode=::NewEnumerator0
    Lifetime=&RandomRangeFloat(Minimum=0.5, Maximum=2)
    Mass Max=10
    Mass Min=1
    Mass Mode=::NewEnumerator3
    Mesh Scale Mode=::NewEnumerator3
    Mesh Uniform Scale=&MakeCustomFloatFromBool(
     Bool=HlslExpression() {hlslExpression="Particles.MeshIndex == 0"}
     True Float=&RandomRangeFloat(Minimum=0.03, Maximum=0.06)
     False Float=&RandomRangeFloat(Minimum=0.1, Maximum=0.13)
    )
    Position Offset=<0 0 -200>
    UsePositionOffset=false
    Write Material Random=true
   )
   &TorusLocation(
    Align Torus With Axis=false
    Coordinate Space:ENiagaraCoordinateSpace=::World
    Large Radius=50
    MaskSpawnGroup=false
    Module.Override Randomness=false
    Offset=<0 0 -100>
   )
   &AddVelocity(
    Coordinate Space=::World
    Velocity=&Add_Vector(
     A=&Multiply_Vector(A=$Output.TorusLocation.TorusNormal, B=&VectorFromFloat(Value=250))
     B=&MakeVector(
      X=0
      Y=0
      Z=&MakeCustomFloatFromBool(
       Bool=HlslExpression() {hlslExpression="User.ForceStrength < 350.0"}
       True Float=600
       False Float=&RandomRangeFloat(Minimum=1100, Maximum=1400)
      )
     )
    )
   ) {enabled=false}
   &InitialMeshOrientation(
    Rotation=&MakeVector(
     X=&RandomRangeFloat(Minimum=0, Maximum=1)
     Y=&RandomRangeFloat(Minimum=0, Maximum=1)
     Z=&RandomRangeFloat(Minimum=0, Maximum=1)
    )
    Use Orientation Vector=true
    Use Rotation Vector=true
   )
   &AddVelocityInCone(
    Cone Angle=80
    Cone Axis=<0 0 1>
    Cone Axis Coordinate Space=::World
    Module.Use Velocity Falloff On Cone Axis=false
    Velocity Strength=&SelectFloatFromArray(Direct Array Index=$Particles.ForcePicker, Float Selection Array=DataInterface() {propertyValues {FloatData [700,750,850,1200,1610]}})
   )
  ]
  particleUpdateScript [
   &Drag(Use Linear Drag=true, Use Rotational Drag=false)
   &GravityForce(Gravity=&MakeVector(X=0, Y=0, Z=&RandomRangeFloat(Minimum=-2300, Maximum=-2550)))
   &Collision(
    Age Colliding Particles=false
    Module.Control Rest State Via Surface Normal=false
    Module.EnableMaxCollisionCount=false
    Module.Randomize Collision Normal Vector=false
   )
   &SolveForcesAndVelocity(Clamp Velocity=false, Limit Acceleration=false)
   &ParticleState()
   &ScaleMeshSize(Scale Factor=&VectorFromFloat(Value=&FloatFromCurve(FloatCurve=DataInterface() {propertyValues {Curve {keys [
          {interpMode=RCIM_Cubic, tangentMode=RCTM_Break, time=0.8, value=1}
          {arriveTangent=-14.1904, interpMode=RCIM_Cubic, tangentMode=RCTM_Break, time=1, value=0}
         ]}}})))
   &MeshRotationForce(Rotation=&RandomRangeVector(Minimum=<-500 -500 -250>, Maximum=<500 500 250>)) {enabled=false}
   &SolveRotationalForcesAndVelocity()
  ]
  rendererClasses [&NiagaraMeshRendererProperties]
  renderers [{rendererClass=&NiagaraMeshRendererProperties, rendererIndex=0}]
 }
 Bounce_Dust001 {
  emitterUpdateScript [
   &EmitterState(Life Cycle Mode=::NewEnumerator1, Loop Behavior=::NewEnumerator1, Loop Delay=0.3, Loop Duration=1.15, UseLoopDelay=false)
   &SpawnBurst_Instantaneous(Spawn Count=10, Use Spawn Probability=false)
  ]
  particleSpawnScript [
   &self:ForceIndex(ForceStrength:NiagaraFloat=$User.ForceStrength) {module=ForceIndex}
   &FN_SetSpawnGroup(Number of Spawn Groups=10)
   &InitializeParticle(
    Lifetime=&RandomRangeFloat(Minimum=0.5, Maximum=1.125)
    Mass Mode=::NewEnumerator1
    Position Offset=<0 0 -50>
    Sprite Rotation Angle=&RandomRangeFloat(Minimum=0, Maximum=360)
    Sprite Rotation Mode=::NewEnumerator0
    Sprite Size Mode=::NewEnumerator3
    Uniform Sprite Size=&SelectFloatFromArray(Direct Array Index=$Particles.ForcePicker, Float Selection Array=DataInterface() {propertyValues {FloatData [70,85,125,150,160]}})
    UsePositionOffset=false
    Write Material Random=true
   )
   &TorusLocation(
    Align Torus With Axis=false
    Coordinate Space:ENiagaraCoordinateSpace=::World
    Large Radius:NiagaraFloat=&SelectFloatFromArray(Direct Array Index=$Particles.ForcePicker, Float Selection Array=DataInterface() {propertyValues {FloatData [20,80,100,120,150]}})
    MaskSpawnGroup=false
    Offset=<0 0 -90>
    Radius Position=0
    Torus Distribution Mode:ENiagaraTorusDistributionMode=::NewEnumerator1
    U Position=HlslExpression() {hlslExpression=Particles.SpawnGroup *.1}
   )
   &AddVelocity(
    Coordinate Space=::World
    Velocity=&Multiply_Vector(A=$Output.TorusLocation.TorusNormal, B=&VectorFromFloat(Value=70))
   )
  ]
  particleUpdateScript [
   &ParticleState()
   &Color(
    Color=&ColorFromCurve(ColorCurve=DataInterface() {propertyValues {
       AlphaCurve {keys [<at 0 1 RCIM_Cubic>,<at 1 1 RCIM_Cubic>]}
       BlueCurve {keys [<at 0 0.3333 RCIM_Cubic>,<at 1 0.4375 RCIM_Cubic>]}
       GreenCurve {keys [<at 0 0.3333 RCIM_Cubic>,<at 1 0.4375 RCIM_Cubic>]}
       RedCurve {keys [<at 0 0.3333 RCIM_Cubic>,<at 1 0.4375 RCIM_Cubic>]}
      }})
    Scale Alpha=&FloatFromCurve(
     FloatCurve=DataInterface() {propertyValues {Curve {keys [
         {interpMode=RCIM_Cubic, leaveTangent=11.125, tangentMode=RCTM_Break, time=0, value=0}
         {interpMode=RCIM_Cubic, leaveTangent=-2.2426, tangentMode=RCTM_Break, time=0.25, value=1.5}
         {interpMode=RCIM_Cubic, tangentMode=RCTM_Break, time=1, value=0}
        ]}}}
     Scale Curve=0.35
    )
   )
   &Drag(
    Drag=&FloatFromCurve(
     FloatCurve=DataInterface() {propertyValues {Curve {keys [
         {interpMode=RCIM_Cubic, tangentMode=RCTM_Break, time=0, value=1}
         {arriveTangent=-21.7864, interpMode=RCIM_Cubic, tangentMode=RCTM_Break, time=0.3, value=5}
        ]}}}
     Scale Curve=&FN_Select_Float_from_Time_Of_Day(Morning=0, Day=0, Evening=0, Night=0)
    )
    Use Linear Drag=true
    Use Rotational Drag=false
   )
   &DynamicMaterialParameters(
    Index 0 Param 1=&Add_Float(
     A=$Particles.RandomFloat
     B=&FloatFromCurve(
      FloatCurve=DataInterface() {propertyValues {Curve {keys [{arriveTangent=2.7533, time=0, value=1}, {arriveTangent=0.0529, leaveTangent=0.2021, time=1, value=1.5}]}}}
      CurveIndex=&Add_Float(A=&Multiply_Float(A=$Particles.Initial.MaterialRandom, B=0.5), B=$Particles.NormalizedAge)
     )
    )
    Index 0 Param 2=&FloatFromCurve(FloatCurve=DataInterface() {propertyValues {Curve {keys [
         {arriveTangent=0.0453, interpMode=RCIM_Cubic, tangentMode=RCTM_Break, time=0, value=-1}
         {arriveTangent=0.0604, interpMode=RCIM_Cubic, leaveTangent=0.0604, tangentMode=RCTM_User, time=1, value=-1}
        ]}}})
    Index 0 Param 3=&FloatFromCurve(FloatCurve=DataInterface() {propertyValues {Curve {keys [
         {arriveTangent=-4.113, interpMode=RCIM_Cubic, leaveTangent=-2.1889, tangentMode=RCTM_User, time=0, value=1}
         {arriveTangent=0.2347, interpMode=RCIM_Cubic, leaveTangent=0.2723, tangentMode=RCTM_Break, time=1, value=0}
        ]}}})
    Index 0 Param 4=&FloatFromCurve(FloatCurve=DataInterface() {propertyValues {Curve {keys [<at 0 1>,<at 1 0>]}}}, Scale Curve=0)
    Index 1 Param 1=&FloatFromCurve(
     FloatCurve=DataInterface() {propertyValues {Curve {keys [
         {interpMode=RCIM_Cubic, leaveTangent=4.7088, tangentMode=RCTM_Break, time=0, value=0}
         {arriveTangent=4.2257, interpMode=RCIM_Cubic, leaveTangent=4.2257, tangentMode=RCTM_User, time=0.25, value=1}
        ]}}}
     Scale Curve=&RandomRangeFloat(Minimum=1.5, Maximum=1.5)
    )
    Index 1 Param 2=&FloatFromCurve(
     FloatCurve=DataInterface() {propertyValues {Curve {keys [<at 0 0>, {leaveTangent=0.9066, time=1, value=1}]}}}
     Scale Curve=&RandomRangeFloat(Minimum=0.5, Maximum=0.6)
    )
    Index 1 Param 3=0.25
    Index 1 Param 4=&FloatFromCurve(
     FloatCurve=DataInterface() {propertyValues {Curve {keys [{arriveTangent=-0.3224, interpMode=RCIM_Cubic, leaveTangent=-0.3224, tangentMode=RCTM_User, time=0, value=1}, <at 1 0.9>]}}}
     Scale Curve=&RandomRangeFloat(Minimum=1.1, Maximum=1.4)
    )
    Index 2 Param 1=&FloatFromCurve(FloatCurve=DataInterface() {propertyValues {Curve {keys [{interpMode=RCIM_Cubic, leaveTangent=1.4673, tangentMode=RCTM_Break, time=0, value=0}, <at 1 1>]}}})
    Index 2 Param 2=&FloatFromCurve(
     FloatCurve=DataInterface() {propertyValues {Curve {keys [
         {arriveTangent=1.9056, interpMode=RCIM_Cubic, leaveTangent=1.9056, tangentMode=RCTM_User, time=0, value=0.2}
         {arriveTangent=0.0259, interpMode=RCIM_Cubic, leaveTangent=0.0259, tangentMode=RCTM_User, time=1, value=1.4}
        ]}}}
     Scale Curve=&RandomRangeFloat(Minimum=0.1, Maximum=0.4)
    )
    Index 2 Param 3=&FloatFromCurve(
     FloatCurve=DataInterface() {propertyValues {Curve {keys [
         {interpMode=RCIM_Cubic, leaveTangent=2.6052, tangentMode=RCTM_Break, time=0, value=1}
         {interpMode=RCIM_Cubic, leaveTangent=11.1233, tangentMode=RCTM_Break, time=1, value=2}
        ]}}}
     Scale Curve=2
    )
    Index 2 Param 4=&FloatFromCurve(FloatCurve=DataInterface() {propertyValues {Curve {keys [{interpMode=RCIM_Cubic, tangentMode=RCTM_Break, time=0, value=1}, {interpMode=RCIM_Cubic, tangentMode=RCTM_Break, time=1, value=1.1}]}}})
    Module.Param0WriteEnabled=true
    Module.Param0WriteEnabled1=false
    Module.Param0WriteEnabled2=true
    Module.Param1WriteEnabled=false
    Module.Param1WriteEnabled1=false
    Module.Param1WriteEnabled2=true
    Module.Param2WriteEnabled=true
    Module.Param2WriteEnabled1=true
    Module.Param2WriteEnabled2=true
    Module.Param3WriteEnabled=true
    Module.Param3WriteEnabled1=true
    Module.Param3WriteEnabled2=true
    Write Parameter Index 1=true
    Write Parameter Index 2=true
   )
   &ScaleSpriteSize(Scale Factor=&Vector2DFromFloat(Value=&FloatFromCurve(FloatCurve=DataInterface() {propertyValues {Curve {keys [
          {interpMode=RCIM_Cubic, leaveTangent=-0.1211, tangentMode=RCTM_Break, time=0, value=0.42}
          {arriveTangent=-0.1422, interpMode=RCIM_Cubic, tangentMode=RCTM_Break, time=0.5, value=2}
         ]}}})))
   &SolveForcesAndVelocity(Clamp Velocity=false, Limit Acceleration=false)
   &SpriteRotationRate(Rotation Rate=&RandomRangeFloat(Minimum=-20, Maximum=20))
  ]
  rendererClasses [&NiagaraSpriteRendererProperties]
  renderers [{rendererClass=&NiagaraSpriteRendererProperties}]
 }
}
summary {
 systemName=NS_CageSaddle_Poofs
 emitters [
  {emitterName=Bounce_Dust001, rendererClasses [&NiagaraSpriteRendererProperties]}
  {emitterName=Bounce_Area_Rocks001, rendererClasses [&NiagaraMeshRendererProperties]}
 ]
 userVariables [
  User.BounceCharge {type=NiagaraFloat, defaultValue=0, description=""}
  User.BounceSmokeActive {type=NiagaraBool, defaultValue=true}
  User.ForceStrength {type=NiagaraFloat, defaultValue=1610, description=""}
  User.InitialSmokeTransformActive {type=NiagaraBool}
  User.SmokeLocation {type=Vector3f, description="", defaultValue {struct=Vector3f, value=<0 0 0>}}
  User.SpeedLinesThreshold {type=NiagaraFloat, defaultValue=0, description=""}
  User.SuckingIn {type=NiagaraBool}
 ]
}
